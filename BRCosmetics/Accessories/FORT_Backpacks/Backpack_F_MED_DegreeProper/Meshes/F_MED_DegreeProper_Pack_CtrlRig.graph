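# Control Rig for the DegreeProper backpack cosmetic that procedurally sways the hanging case bone, dyn_case_base, in response to character motion.
# Skydive and Spine rot inputs are Kalman-smoothed, and the larger of Skydive and doubled Spine rot becomes the maximum of a noisy Item Harmonics wave.
# A second, similar harmonics node driven only by the spine value is not connected to execution.
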ControlRigBlueprint /BRCosmetics/Accessories/FORT_Backpacks/Backpack_F_MED_DegreeProper/Meshes/F_MED_DegreeProper_Pack_CtrlRig
graphs [
 RigVMFunctionLibrary {links [], nodes []}
 RigVMModel {nodes [
   RigUnit_BeginExecution = &RigUnit_BeginExecution()
   KalmanFilter = &RigVMFunction_KalmanFloat(BufferSize=8, Value=RigVMVariableNode() {variable=Skydive}.Value) {template="KalmanFilter::Execute(in Value,in BufferSize,out Result)"}
   Multiply = &RigVMFunction_MathDoubleMul(
    B=2
    A=&RigVMFunction_KalmanFloat(BufferSize=8, Value=RigVMVariableNode() {variable=Spine rot}.Value) {template="KalmanFilter::Execute(in Value,in BufferSize,out Result)"}.Result
   )
   Greater = /Script/RigVM.RigVMFunction_MathDoubleGreater(A=@KalmanFilter.Result, B=@Multiply.Result) {is=struct RigVMUnitNode, template="Greater::Execute(in A,in B,out Result)"}
   /Script/ControlRig.RigUnit_ItemHarmonics(
    RotationOrder=YZX
    Targets=((Item=(Type=Bone,Name="dyn_case_base"),Ratio=0.3))
    WaveAmplitude=(X=0,Y=0,Z=40)
    WaveEase=Linear
    WaveFrequency=(X=0,Y=0.6,Z=22)
    WaveMinimum=0
    WaveNoise=(X=0.3,Y=0,Z=0.3)
    WaveOffset=(X=0,Y=1,Z=2)
    WaveSpeed=(X=4,Y=4,Z=4)
    WaveMaximum=RigVMDispatchNode(Condition=@Greater.Result, True=@KalmanFilter.Result, False=@Multiply.Result) {template="DISPATCH_RigVMDispatch_If(in Condition,in True,in False,out Result)"}.Result
    ExecutePin=@RigUnit_BeginExecution.ExecutePin
   ) {is=struct RigVMUnitNode}
   &RigUnit_ItemHarmonics(
    ExecutePin=()
    RotationOrder=YZX
    Targets=((Item=(Type=Bone,Name="dyn_case_base"),Ratio=0.3))
    WaveAmplitude=(X=0,Y=0,Z=40)
    WaveEase=Linear
    WaveFrequency=(X=22,Y=0.6,Z=22)
    WaveMinimum=0
    WaveNoise=(X=0.3,Y=0,Z=0.3)
    WaveOffset=(X=0,Y=1,Z=2)
    WaveSpeed=(X=4,Y=4,Z=4)
    WaveMaximum=@Multiply.Result
   )
  ]}
]
hierarchy [
 root {loc=<0 0 0>, rot=<0 0 0 1>}
 pack {parent=root, loc=<0 -0.7316 0>, rot=<0 0 0 1>}
 dyn_case_base {parent=pack, loc=<-1.3982 -4.6882 -12.7856>, rot=<0.0207 -0.1897 -0.0038 -0.9816>}
 dyn_case {parent=dyn_case_base, loc=<0 0 0>, rot=<0 0 0 1>}
]
variables [
 Skydive {type=double, default=0}
 Spine rot {type=double, default=0}
]
