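# Physics asset for the DegreeProper backpack cosmetic, defining bodies and joints for its dynamic case and clasp bones.
# Each bone gets a single sphere: a large offset one for the case, a tiny one for its base and three small ones for the clasps.
# The case joins the base through a limited single-axis swing, and each clasp joins the case with a similar near-hinge constraint allowing slight off-axis give.
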PhysicsAsset /BRCosmetics/Accessories/FORT_Backpacks/Backpack_F_MED_DegreeProper/Meshes/F_MED_DegreeProper_Pack_Physics
bodies {
 dyn_case [{shapeType=Sphere, radius=6.8607, shapeName=dyn_case_sphere, center=<4.1355 0 0>}]
 dyn_case_base [{shapeType=Sphere, radius=0.505, shapeName=dyn_case_base_sphere}]
 dyn_clasp_01 [{shapeType=Sphere, radius=0.9022, shapeName=dyn_clasp_01_sphere, center=<0.4952 0.0495 -0.0705>}]
 dyn_clasp_02 [{shapeType=Sphere, radius=0.9022, shapeName=dyn_clasp_02_sphere, center=<0.5483 0.0181 -0.0466>}]
 dyn_clasp_03 [{shapeType=Sphere, radius=0.9022, shapeName=dyn_clasp_03_sphere, center=<0.5518 0.0114 -0.0242>}]
}
constraints [
 {bone1Name=dyn_case, bone2Name=dyn_case_base, swing1LimitDegrees=20, swing2LimitDegrees=0, twistLimitDegrees=0}
 {bone1Name=dyn_clasp_03, bone2Name=dyn_case, swing1LimitDegrees=20, swing2LimitDegrees=1, twistLimitDegrees=1}
 {bone1Name=dyn_clasp_02, bone2Name=dyn_case, swing1LimitDegrees=20, swing2LimitDegrees=1, twistLimitDegrees=1}
 {bone1Name=dyn_clasp_01, bone2Name=dyn_case, swing1LimitDegrees=20, swing2LimitDegrees=1, twistLimitDegrees=1}
]
