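# Animation Blueprint for the DegreeProper backpack cosmetic, posing its pack and case around the wearer's movement.
# Each update maps spine and upper-arm deltas from ref pose into alphas, taking the larger of spine roll or left arm for the case.
# These, plus jump and skydive states, drive a movement-gated Control Rig and additive bone offsets.
# AnimDynamics then swings clasps, tape and handle, with RigidBody physics only during non-emote front-end idle.
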AnimBlueprint /BRCosmetics/Accessories/FORT_Backpacks/Backpack_F_MED_DegreeProper/Meshes/F_MED_DegreeProper_Pack_AnimBP
target_skeleton=&/F_MED_DegreeProper_Pack_Skeleton
graphs [
 AnimGraph {nodes [
   "Transform (Modify) Bone - Bone: root"(Translation="0,0,0", Rotation="0,0,0", Scale="1.2,1.2,1.2") {
    rotationMode=BMM_Additive
    rotationSpace=BCS_BoneSpace
    scaleMode=BMM_Additive
    translationMode=BMM_Additive
    translationSpace=BCS_BoneSpace
    alphaBoolBlend {blendOption=Cubic}
    boneToModify {boneName=root}
    scale=<1.2 1.2 1.2>
   }
   n1 = Control Rig!(
    Source=Input Pose()
    bAlphaBoolEnabled:bool=Get bIsMoving2D()
    Skydive:double=Get Is skydive() -> (Is skydive:double)
    Spine rot:double=Get case logic() -> (case logic:double)
   ) -> (Pose) {
    alphaInputType=Bool
    alphaBoolBlend {blendOption=Linear}
    controlRigAssetReference {blueprintRigClass=&/F_MED_DegreeProper_Pack_CtrlRig.F_MED_DegreeProper_Pack_CtrlRig_C}
   }
   n2 = "Transform (Modify) Bone - Bone: root"(ComponentPose=Local To Component(LocalPose=@n1), Translation="0,-1,0", Rotation="0,0,50") {
    rotationMode=BMM_Additive
    rotationSpace=BCS_BoneSpace
    translationMode=BMM_Additive
    translationSpace=BCS_BoneSpace
    alphaBoolBlend {blendOption=Cubic}
    boneToModify {boneName=root}
    rotation {roll=50}
    translation {y=-1}
   }
   n3 = "Transform (Modify) Bone - Bone: root"(ComponentPose=@n2, Translation="0,-1,0", Rotation="0,0,0") {
    rotationMode=BMM_Additive
    rotationSpace=BCS_BoneSpace
    translationMode=BMM_Additive
    translationSpace=BCS_BoneSpace
    alphaBoolBlend {blendOption=Cubic}
    boneToModify {boneName=root}
    translation {y=-1}
   }
   n4 = "Transform (Modify) Bone - Bone: dyn_case_base"(ComponentPose=@n3, Alpha=Get case logic(), Translation="0,0,0", Rotation="0,-80,0", Scale="1,1,1") {
    rotationMode=BMM_Additive
    rotationSpace=BCS_BoneSpace
    alphaBoolBlend {blendOption=Linear}
    alphaScaleBiasClamp {bInterpResult=true, inRange {min=-1}}
    boneToModify {boneName=dyn_case_base}
    rotation {yaw=-80}
   }
   n5 = "Transform (Modify) Bone - Bone: dyn_case_base"!(ComponentPose=@n4, bAlphaBoolEnabled:bool=Get Is jumping() -> (Is jumping:bool), Translation="0,0,0", Rotation="0,-40,0", Scale="1,1,1") -> (Pose) {
    alphaInputType=Bool
    rotationMode=BMM_Additive
    rotationSpace=BCS_BoneSpace
    alphaBoolBlend {blendInTime=0.5, blendOutTime=0.5}
    alphaScaleBiasClamp {bInterpResult=true, inRange {min=-1}}
    boneToModify {boneName=dyn_case_base}
    rotation {yaw=-40}
   }
   n6 = "Transform (Modify) Bone - Bone: pack"(ComponentPose=@n5, Alpha=Get upperarm r rot() -> (upperarm r rot:double), Translation="0,0,0", Rotation="0,30,0") {
    rotationMode=BMM_Additive
    rotationSpace=BCS_BoneSpace
    alphaBoolBlend {blendOption=Linear}
    alphaScaleBiasClamp {bInterpResult=true, inRange {min=-1}}
    boneToModify {boneName=pack}
    rotation {yaw=30}
   }
   n7 = Anim Dynamics - Bone: dyn_clasp_01(ComponentPose=@n6, Alpha=1, ExternalForce="0,0,0") {
    simulationSpace=BoneRelative
    alphaBoolBlend {blendOption=Linear}
    boundBone {boneName=dyn_clasp_01}
    physicsBodyDefinitions [{
      boundBone {boneName=dyn_clasp_01}
      boxExtents=<10 10 10>
      constraintSetup {angularLimitsMax=<0 0 10>, angularLimitsMin=<0 0 -40>, angularTarget=<0 0 0>}
      localJointOffset=<1 0 0>
     }]
    relativeSpaceBone {boneName=dyn_case}
   }
   n8 = Anim Dynamics - Bone: dyn_clasp_02(ComponentPose=@n7, Alpha=1, ExternalForce="0,0,0") {
    simulationSpace=BoneRelative
    alphaBoolBlend {blendOption=Linear}
    boundBone {boneName=dyn_clasp_02}
    physicsBodyDefinitions [{
      boundBone {boneName=dyn_clasp_02}
      boxExtents=<10 10 10>
      constraintSetup {angularLimitsMax=<0 0 10>, angularLimitsMin=<0 0 -40>, angularTarget=<0 0 0>}
      localJointOffset=<1 0 0>
     }]
    relativeSpaceBone {boneName=dyn_case}
   }
   n9 = Anim Dynamics - Bone: dyn_clasp_03(ComponentPose=@n8, Alpha=1, ExternalForce="0,0,0") {
    simulationSpace=BoneRelative
    alphaBoolBlend {blendOption=Linear}
    boundBone {boneName=dyn_clasp_03}
    physicsBodyDefinitions [{
      boundBone {boneName=dyn_clasp_03}
      boxExtents=<10 10 10>
      constraintSetup {angularLimitsMax=<0 0 10>, angularLimitsMin=<0 0 -40>, angularTarget=<0 0 0>}
      localJointOffset=<1 0 0>
     }]
    relativeSpaceBone {boneName=dyn_case}
   }
   n10 = Anim Dynamics - Bone: dyn_tape(ComponentPose=@n9, Alpha=1, ExternalForce="0,0,0") {
    simulationSpace=BoneRelative
    alphaBoolBlend {blendOption=Linear}
    boundBone {boneName=dyn_tape}
    physicsBodyDefinitions [{
      boundBone {boneName=dyn_tape}
      boxExtents=<10 10 10>
      constraintSetup {angularLimitsMax=<0 0 30>, angularLimitsMin=<0 0 0>, angularTarget=<0 0 0>}
      localJointOffset=<-1 0 0>
     }]
    relativeSpaceBone {boneName=dyn_case}
   }
   n11 = Anim Dynamics - Bone: dyn_handle!(ComponentPose=@n10, Alpha=1, GravityOverride:Vector="0,0,0", ExternalForce="0,0,0") -> (Pose) {
    angularDampingOverride=0.999
    bOverrideAngularDamping=true
    bUseGravityOverride=true
    simulationSpace=BoneRelative
    alphaBoolBlend {blendOption=Linear}
    boundBone {boneName=dyn_handle}
    physicsBodyDefinitions [{
      boundBone {boneName=dyn_handle}
      boxExtents=<10 10 10>
      constraintSetup {angularLimitsMax=<0 0 20>, angularLimitsMin=<0 0 -20>, angularTarget=<0 0 0>}
      localJointOffset=<4 0 0>
     }]
    relativeSpaceBone {boneName=pack}
   }
   n12 = RigidBody!(ComponentPose=@n11, bAlphaBoolEnabled:bool=Get FE Idle() -> (FE Idle:bool), OverrideWorldGravity:Vector="0,0,-980", ExternalForce) -> (Pose) {
    alphaInputType=Bool
    bOverrideWorldGravity=true
    overridePhysicsAsset=&/F_MED_DegreeProper_Pack_Physics
    simulationSpace=BaseBoneSpace
    alphaBoolBlend {blendOption=Linear}
    baseBoneRef {boneName=dyn_case_base}
    componentAppliedLinearAccClamp=<1500 1500 1500>
    componentLinearAccScale=<0.4 0.4 0.4>
    componentLinearVelScale=<0.5 0.5 0.5>
    overrideWorldGravity {z=-980}
    simSpaceSettings {velocityScaleZ=0.1}
   }
   Output Pose(Result=Component To Local(ComponentPose=@n12)) {layerGroup=None}
   Get spine rot() -> (spine rot:double)
  ]}
 EventGraph {nodes [
   Gravity Override State Function DEPRECATED(CharacterPartAnimInstance=Self-Reference(), GameTimeInSeconds=GetGameTimeInSeconds(), GravityOverrideParams=Get Gravity Override Params())
   > Set Gravity Override(Gravity Override="0, 0, 0")
   Event BlueprintUpdateAnimation()
   > n1 = Sequence!(execute) -> (then_0, then_1, then_2:Exec, then_3:Exec)
   > Delay(Duration=0.2)
   n2 = GetDeltaTransformFromRefPose!(self=GetOwnerSkeletalMeshComponent(), BoneName=spine_05, BaseName=pelvis) -> (
    ReturnValue_Location:Vector
    ReturnValue_Rotation_Roll:float
    ReturnValue_Rotation_Pitch:float
    ReturnValue_Rotation_Yaw:float
    ReturnValue_Scale:Vector
   )
   n3 = MapRangeClamped(Value=@n2.ReturnValue_Rotation_Roll, InRangeA=0, InRangeB=-70)
   n4 = MapRangeClamped(Value=@n2.ReturnValue_Rotation_Roll, InRangeA=0, InRangeB=70)
   > n5 = Branch(Condition="float > float"(A=@n4, B=@n3))
   > Set spine rot(spine rot:double=@n4) -> (Output_Get:double)
   Set spine rot(execute=@n5.else, spine rot=@n3)
   Set Is skydive(execute=@n1.then_1, Is skydive:double="To Float (Boolean)"(InBool=Get bIsSkyDiving())) -> (Output_Get:double)
   n6 = Get bIsPlayingEmote()
   > Set FE Idle(FE Idle:bool=AND Boolean(A=Get bIsInFrontEnd(), B=NOT Boolean(A=@n6))) -> (Output_Get:bool)
   n7 = Gravity Override State Function DEPRECATED(
    execute=@n1.then_2
    CharacterPartAnimInstance=Self-Reference()
    GameTimeInSeconds=GetGameTimeInSeconds()
    GravityOverrideParams=Get Gravity Override Params()
   )
   > Set Gravity Override(Gravity Override=@n7)
   n8 = GetOwnerSkeletalMeshComponent()
   Set Spine rotation(
    execute=@n1.then_3
    Spine rotation:double=MapRangeClamped(Value=GetDeltaTransformFromRefPose(self=@n8, BoneName=spine_05, BaseName=pelvis).ReturnValue_Rotation_Yaw, InRangeA=10, InRangeB=-90)
   ) -> (Output_Get:double)
   n9 = GetOwnerSkeletalMeshComponent()
   > Set upperarm l rot(upperarm l rot:double=MapRangeClamped(Value=GetDeltaTransformFromRefPose(self=@n9, BoneName=upperarm_l, BaseName=spine_05).ReturnValue_Rotation_Yaw, InRangeA=10, InRangeB=-120)) -> (Output_Get:double)
   n10 = GetOwnerSkeletalMeshComponent()
   > Set upperarm r rot(upperarm r rot:double=MapRangeClamped(Value=GetDeltaTransformFromRefPose(self=@n10, BoneName=upperarm_r, BaseName=spine_05).ReturnValue_Rotation_Yaw, InRangeA=10, InRangeB=-120)) -> (Output_Get:double)
   n11 = Get upperarm l rot() -> (upperarm l rot:double)
   n12 = Get spine rot()
   > n13 = Branch(Condition="float > float"(A=@n12, B=@n11))
   > Set case logic(case logic:double=@n12) -> (Output_Get:double)
   Set case logic(execute=@n13.else, case logic=@n11)
   Event OnBeginJumping()
   > Delay(Duration=0.2)
   > Set Is jumping(Is jumping:bool=true) -> (Output_Get:bool)
   n14 = Event OnCharacterCustomizationLoaded()
   > Set Body_Mesh(Body_Mesh:Skeletal Mesh Component Budgeted Object Reference=Get Array Item with EFort Custom Part Type(Array=@n14.CustomMeshParts)) -> (Output_Get:Skeletal Mesh Component Budgeted Object Reference)
   Event OnBeginFalling()
   > Delay(Duration=0.2)
   > Set Is jumping(Is jumping=false)
   RetriggerableDelay(Duration=0.2)
   TryGetPawnOwner()
   Event OnCharacterAndPickaxeCustomizationLoaded!() -> (
    OutputDelegate
    then
    CustomBaseMesh:Skeletal Mesh Component Object Reference
    CustomParts:Array of Custom Character Part Object References
    CustomMeshParts:Array of Skeletal Mesh Component Budgeted Object References
   ) {is=title K2Node_Event}
   "float > float"(A=@n4, B="To Float (Boolean)"(InBool=Get bIsJumping()))
   Get bIsFalling()
  ]}
]
variables [
 Body_Mesh {type=SkeletalMeshComponentBudgeted}
 case logic {type=number}
 Delta Time X {type=number}
 FE Idle {type=boolean}
 Gravity Override {type=Vector}
 Gravity Override Params {type=GravityOverrideParamsStruct, default {onGround_Moving {gravityOverride=<0 0 980>}}}
 Is jumping {type=boolean}
 Is skydive {type=number}
 spine rot {type=number}
 Spine rotation {type=number}
 upperarm l rot {type=number}
 upperarm r rot {type=number}
]
